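; AI controller for BuzzCool clone NPCs that reacts to goal actor changes.
; Player goals activate the clone and start timed polling of leash and state; large donut goals register it with the donut's processor and may apply an idle cooldown.
; On authority, the leash follows the goal and pawn state reflects goal type and combat token.
; An activated clone losing its goal is force-killed after a timeout; distance notifies update blackboard range flags.
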(event EventBeginPlay
  (Development|PrintString "@@@ Event BeginPlay [AIController_BuzzCool_Clone]: CALLED!" false (Variables|Debug|GetbDebugController))
  (bind _returnvalue (AI|GoalSelection|GetAIGoalComponent self))
  (EventDispatchers|BindEventtoOnNewGoalActorDelegate _returnvalue (EventDispatchers|CreateEvent :!event Custom|OnCloneGoalActorChanged))
  (AI|Components|Blackboard|SetValueasBool (AI|GetBlackboard self) (Variables|Internal|GetGoalActorIsHostile_BBKey) false)
  (Variables|Internal|SetbCloneIsActivated false))

(event Custom|OnCloneGoalActorChanged
  (params (OldActor :type Actor%20Object%20Reference) (NewActor :type Actor%20Object%20Reference))
  (Development|PrintString (Utilities|String|Append "@@@@@@ OnCloneGoalActorChanged [AIController_BuzzCool_Clone]: CALLED; Goal Actor changed from: " (Utilities|GetDisplayName OldActor) ", to: " (Utilities|GetDisplayName NewActor) "!") false (Variables|Debug|GetbDebugController))
  (Utilities|IsValid NewActor
    (:Is%20Valid
      (CallFunction|StopPawnDeathPolling)
      (CallFunction|ProcessActorTarget NewActor)
      (if (Variables|Internal|GetbGoalIsAPawn)
        (Variables|Internal|SetCurrentGoalPawnActor NewActor)
        (CallFunction|StartPawnActorPolling)
        (else
          (CallFunction|StopPawnActorPolling)
          (CallFunction|UpdateCloneLeash NewActor)
          (CallFunction|UpdateCloneState NewActor))))
    (:Is%20Not%20Valid
      (Variables|Internal|SetCurrentGoalPawnActor 0)
      (CallFunction|StopPawnActorPolling)
      (CallFunction|UpdateCloneLeash 0)
      (CallFunction|UpdateCloneState 0)
      (if (Variables|Internal|GetbCloneIsActivated)
        (CallFunction|StartPawnDeathPolling)
        (Variables|Internal|SetbCloneIsActivated false)))))

(event Custom|StartPawnDeathPolling
  (Development|PrintString "@@@ StartPawnDeathPolling [AIController_BuzzCool_Clone]: CALLED!" false (Variables|Debug|GetbDebugController))
  (if (not (Utilities|Time|IsTimerActivebyHandle (Variables|Internal|GetTimerHandle_PawnDeathUpdate)))
    (bind _returnvalue (Utilities|Time|SetTimerbyEvent (delegate Custom|OnCollectedPawnTimeout) (ScalableFloat|GetValueatLevel (Variables|Internal|GetSF_CollectedPawnTimeout))))
    (Variables|Internal|SetTimerHandle_PawnDeathUpdate _returnvalue)))

(event Custom|StartPawnActorPolling
  (bind _currentgoalpawnactor (Variables|Internal|GetCurrentGoalPawnActor))
  (Development|PrintString "@@@ StartPawnActorPolling [AIController_BuzzCool_Clone]: CALLED!" false (Variables|Debug|GetbDebugController))
  (if (not (Utilities|Time|IsTimerActivebyHandle (Variables|Internal|GetTimerHandle_PawnActorUpdate)))
    (CallFunction|UpdateCloneLeash _currentgoalpawnactor)
    (CallFunction|UpdateCloneState _currentgoalpawnactor)
    (bind _returnvalue (Utilities|Time|SetTimerbyEvent (delegate Custom|UpdateItemOwnerData) (ScalableFloat|GetValueatLevel (Variables|Internal|GetSF_ItemOwnerPollTime)) true))
    (Variables|Internal|SetTimerHandle_PawnActorUpdate _returnvalue)))

(event Custom|OnCollectedPawnTimeout
  (bind _self self)
  (Development|PrintString "@@@@@@ OnCollectedPawnTimeout [AIController_BuzzCool_Clone]: CALLED!" false (Variables|Debug|GetbDebugController))
  (bind _asfort_pawn (Utilities|Casting|CastToFortPawn (Pawn|GetControlledPawn _self)))
  (FortPawn|ForceKill _asfort_pawn "(TagName=\"Gameplay.Effect.InstantDeath.Despawn\")" _self _asfort_pawn))

(event Custom|UpdateItemOwnerData
  (bind _currentgoalpawnactor (Variables|Internal|GetCurrentGoalPawnActor))
  (Development|PrintString "@@@@@@ UpdateItemOwnerData [AIController_BuzzCool_Clone]: CALLED!" false (Variables|Debug|GetbDebugController))
  (CallFunction|UpdateCloneLeash _currentgoalpawnactor)
  (CallFunction|UpdateCloneState _currentgoalpawnactor))

(event Custom|StopPawnActorPolling
  (Development|PrintString "@@@ StopPawnActorPolling [AIController_BuzzCool_Clone]: CALLED!" false (Variables|Debug|GetbDebugController))
  (Utilities|Time|ClearandInvalidateTimerbyHandle (Variables|Internal|GetTimerHandle_PawnActorUpdate)))

(event Custom|StopPawnDeathPolling
  (Development|PrintString "@@@ StopPawnDeathPolling [AIController_BuzzCool_Clone]: CALLED!" false (Variables|Debug|GetbDebugController))
  (if (Utilities|Time|IsTimerActivebyHandle (Variables|Internal|GetTimerHandle_PawnDeathUpdate))
    (Utilities|Time|ClearandInvalidateTimerbyHandle (Variables|Internal|GetTimerHandle_PawnDeathUpdate))))

(event EventOnGoalActorDistanceNotify
  (params (NotifyIndex :type Integer) (IsValid :type Boolean))
  (switch int NotifyIndex
    (:0
      (if (not (== IsValid (Variables|Internal|GetGoalActorDistance_IsInMeleeRange)))
        (bind _output_get (Variables|Internal|SetGoalActorDistance_IsInMeleeRange IsValid))
        (AI|Components|Blackboard|SetValueasBool (AI|GetBlackboard self) (Variables|Internal|GetGoalActorDistance_IsInMeleeRangeKey) _output_get)))
    (:1
      (if (not (== IsValid (Variables|Internal|GetGoalActorDistance_IsInCloseRange)))
        (bind _output_get_2 (Variables|Internal|SetGoalActorDistance_IsInCloseRange IsValid))
        (AI|Components|Blackboard|SetValueasBool (AI|GetBlackboard self) (Variables|Internal|GetGoalActorDistance_IsInCloseRangeKey) _output_get_2)))))

(fn ProcessActorTarget
  (params (TargetActor :type Actor%20Object%20Reference))
  (bind _targetactor (Variables|GetTargetActor))
  (bind _targetactor_2 (Variables|GetTargetActor))
  (Development|PrintString "@@@@@@@@@ ProcessActorTarget [AIController_BuzzCool_Clone]: CALLED!" false (Variables|Debug|GetbDebugController))
  (Utilities|IsA(soft) _targetactor "/Script/FortniteGame.FortPlayerPawn"
    (:True
      (Variables|Internal|SetbGoalIsAPawn true)
      (bind _asfort_player_pawn (Utilities|Casting|CastToFortPlayerPawn _targetactor)
        (:then
          (LocalVariables|SetCurrentGoalPawn _asfort_player_pawn)
          (Variables|Internal|SetbCloneIsActivated true))
        (:CastFailed
          _)))
    (:False
      (Variables|Internal|SetbGoalIsAPawn false)
      (LocalVariables|SetCurrentGoalPawn 0)
      (Utilities|IsA(soft) _targetactor_2 "/BuzzCool/Gameplay/BP_BuzzCool_Donut_Large.BP_BuzzCool_Donut_Large_C"
        (:True
          (if (Utilities|DoesObjectImplementInterface _targetactor_2 "/BuzzCool/Interfaces/BuzzCoolDonut_Interface.BuzzCoolDonut_Interface_C")
            (bind _clonespawnactorprocessor (Class|BuzzCoolDonutInterface|GetSpawnActorProcessor(Message) _targetactor_2))
            (bind _aspawn_wild_estate_buzz_cool (Utilities|Casting|CastToPawn_WildEstate_BuzzCool (Pawn|GetControlledPawn self))
              (:then
                (Class|BuzzCoolProcessorInterface|AddActiveClone(Message) _clonespawnactorprocessor _aspawn_wild_estate_buzz_cool)
                (AI|Components|Blackboard|SetValueasObject (AI|GetBlackboard self) (Variables|Internal|GetGoalActorProcessor_BBKey) _clonespawnactorprocessor)
                (if (<= (Math|Random|RandomFloatinRange 0.0 1.0) (ScalableFloat|GetValueatLevel (Variables|Internal|GetSF_LocatioIdleChance)))
                  (GameplayEffects|ApplyGameplayEffectToTarget :self (Ability|GetAbilitySystemComponent (Pawn|GetControlledPawn self)) :GameplayEffectClass "/BuzzCool/NPCs/Abilities/GE_WildEstate_Location_Idle_Cooldown.GE_WildEstate_Location_Idle_Cooldown_C" :Target (Ability|GetAbilitySystemComponent (Pawn|GetControlledPawn self)) :Level 0.0)
                  (else
                    _))
                (bind _output_get (Variables|Internal|SetbHasHostileGoal (and (Variables|Internal|GetbGoalIsAPawn) (Utilities|Enum|Equal(Enum) (Team|GetTeamAffiliation (Pawn|GetControlledPawn self) (LocalVariables|GetCurrentGoalPawn)) "Hostile"))))
                (AI|Components|Blackboard|SetValueasBool (AI|GetBlackboard self) (Variables|Internal|GetGoalActorIsHostile_BBKey) _output_get))
              (:CastFailed
                _))
            (else
              _)))
        (:False
          _)))))

(fn UpdateCloneLeash
  (params (LeashToActor :type Actor%20Object%20Reference))
  (bind _fortathenaleash (Variables|Components|GetFortAthenaLeash))
  (bind _fortathenaleash_2 (Variables|Components|GetFortAthenaLeash))
  (Networking|SwitchHasAuthority
    (:Authority
      (Utilities|IsValid LeashToActor
        (:Is%20Valid
          (Class|FortAthenaLeashComponent|SetLeashCenter _fortathenaleash_2 (Transformation|GetActorLocation LeashToActor)))
        (:Is%20Not%20Valid
          (Class|FortAthenaLeashComponent|SetLeashCenter _fortathenaleash_2 (Pawn|GetNavAgentLocation (Pawn|GetControlledPawn self)))))
      (Class|FortAthenaLeashComponent|SetLeashRadius _fortathenaleash_2 (ScalableFloat|GetValueatLevel (Variables|Internal|GetSF_ActorLeash_Radius)))
      (if (Variables|Debug|GetbDebugController)
        (Utilities|IsValid _fortathenaleash
          (:Is%20Valid
            (Rendering|Debug|DrawDebugSphere (Class|FortAthenaLeashComponent|GetLeashCenter _fortathenaleash) (Class|FortAthenaLeashComponent|GetLeashRadius _fortathenaleash) 50 "(R=0.978654,G=0.019030,B=1.000000,A=1.000000)" 5.0 5.0)))))))

(fn UpdateCloneState
  (params (TargetActor :type Actor%20Object%20Reference))
  (bind _targetactor (Variables|GetTargetActor))
  (Networking|SwitchHasAuthority
    (:Authority
      (Utilities|IsA(soft) _targetactor "/Script/FortniteGame.FortPlayerPawn"
        (:True
          (Utilities|IsValid _targetactor
            (:Is%20Valid
              (bind _returnvalue (GameplayTags|HasMatchingGameplayTag(Message) (Pawn|GetControlledPawn self) "(TagName=\"BuzzCool.AI.Behaviors.Token.HasCombat\")"))
              (if _returnvalue
                (bind _output_get (Variables|Internal|SetEnum_PawnGoalState "NewEnumerator2"))
                (LocalVariables|Setloc_PawnGoalState _output_get)
                (else
                  (bind _output_get_2 (Variables|Internal|SetEnum_PawnGoalState "NewEnumerator3"))
                  (LocalVariables|Setloc_PawnGoalState _output_get_2))))
            (:Is%20Not%20Valid
              (bind _output_get_3 (Variables|Internal|SetEnum_PawnGoalState "NewEnumerator5"))
              (LocalVariables|Setloc_PawnGoalState _output_get_3))))
        (:False
          (Utilities|IsA(soft) (Variables|GetTargetActor) "/BuzzCool/Gameplay/BP_BuzzCool_Donut_Large.BP_BuzzCool_Donut_Large_C"
            (:True
              (bind _output_get_4 (Variables|Internal|SetEnum_PawnGoalState "NewEnumerator4"))
              (LocalVariables|Setloc_PawnGoalState _output_get_4))
            (:False
              (bind _output_get_5 (Variables|Internal|SetEnum_PawnGoalState "NewEnumerator5"))
              (LocalVariables|Setloc_PawnGoalState _output_get_5)))))
      (Development|PrintString (Utilities|String|Append "@@@@@@@@@ UpdateCloneState [AIController_BuzzCool_Clone]: CALLED - Pawn Goal State = " (Utilities|String|EnumtoString (LocalVariables|Getloc_PawnGoalState)) "!") false (Variables|Debug|GetbDebugController))
      (bind _aspawn_wild_estate_buzz_cool (Utilities|Casting|CastToPawn_WildEstate_BuzzCool (Pawn|GetControlledPawn self)))
      (Class|PawnWildEstateBuzzCool|SetBuzzCoolState _aspawn_wild_estate_buzz_cool (LocalVariables|Getloc_PawnGoalState)))))

(fn ConstructionScript)
